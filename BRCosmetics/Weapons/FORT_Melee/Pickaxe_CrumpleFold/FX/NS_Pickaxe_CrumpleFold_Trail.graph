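# Niagara system for the swing trail of the CrumpleFold pickaxe, built from three ribbon emitters: SoftTrail, ToonTrail and WireStreaks.
# Each spawns, initializes and updates ribbon particles through shared AnimTrail functions that read bone or socket pairs on a system-level skeletal mesh.
# All fade alpha by a system swing-fade value times a curve, with WireStreaks' curve indexed by system age.
# SoftTrail and ToonTrail also tint by a TOD color, and SoftTrail tapers its width.
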NiagaraSystem /BRCosmetics/Weapons/FORT_Melee/Pickaxe_CrumpleFold/FX/NS_Pickaxe_CrumpleFold_Trail
dependencies {
 usedDataInterfaces [&NiagaraDataInterfaceCurve, &NiagaraDataInterfaceSkeletalMesh, &NiagaraDataInterfaceColorCurve]
 usedDynamicInputs [&ColorFromCurve]
 usedModules [
  &self:SystemSpawnScript.SystemScriptSource.SystemScriptGraph.NiagaraNodeAssignment_3.SetVariables_#1
  &SystemState
  &self:SystemSpawnScript.SystemScriptSource.SystemScriptGraph.NiagaraNodeAssignment_4.SetVariables_#2
  &self:SystemSpawnScript.SystemScriptSource.SystemScriptGraph.NiagaraNodeAssignment_0.SetVariables_#3
 ]
}
emitters {
 SoftTrail {
  emitterUpdateScript [
   &EmitterState()
   &FN_SpawnAnimTrail()
  ]
  particleSpawnScript [
   &FN_InitializeAnimTrailBoneIndecies(First Bone or Socket Index=1, Second Bone or Socket Index=0)
   &FN_InitializeAnimTrail(AnimTrailWidth=1.2, Color=<rgba 0.7 0.434 0.28 1>, Lifetime=0, SkeletalMesh=$System.Skeletal Mesh)
  ]
  particleUpdateScript [
   &SolveForcesAndVelocity(Clamp Velocity=false, Limit Acceleration=false)
   &FN_UpdateAnimTrail(Locked Particle RibbonWidth=1, SkeletalMesh=$System.Skeletal Mesh) {module=FN_UpdateAnimTrail001}
   &ScaleColor(
    Scale Alpha=&Multiply_Float(
     A=$System.SwingGeneralFade
     B=&FloatFromCurve(FloatCurve=DataInterface() {propertyValues {Curve {keys [{arriveTangent=-0.0783, interpMode=RCIM_Cubic, leaveTangent=-0.0783, tangentMode=RCTM_User, time=0.9, value=0.9}, <at 1 0 RCIM_Cubic>]}}})
    )
    Scale RGB=&MakeVectorFromLinearColorRGB(Color=$System.TODColor)
    ScaleA=true
    ScaleRGB=true
   ) {module=ScaleColor001}
   &ScaleRibbonWidth(Ribbon Width Scale=&FloatFromCurve(FloatCurve=DataInterface() {propertyValues {Curve {keys [
         {arriveTangent=0.3072, interpMode=RCIM_Cubic, leaveTangent=0.3072, tangentMode=RCTM_User, time=0, value=1}
         {arriveTangent=-0.7535, interpMode=RCIM_Cubic, leaveTangent=-0.7535, tangentMode=RCTM_User, time=1, value=0}
        ]}}}))
  ]
  rendererClasses [&NiagaraRibbonRendererProperties]
  renderers [{rendererClass=&NiagaraRibbonRendererProperties}]
 }
 ToonTrail {
  emitterUpdateScript [
   &EmitterState()
   &FN_SpawnAnimTrail()
  ]
  particleSpawnScript [
   &FN_InitializeAnimTrailBoneIndecies(First Bone or Socket Index=1, Second Bone or Socket Index=0)
   &FN_InitializeAnimTrail(AnimTrailWidth=1.2, Lifetime=0.1, SkeletalMesh=$System.Skeletal Mesh)
  ]
  particleUpdateScript [
   &SolveForcesAndVelocity(Clamp Velocity=false, Limit Acceleration=false)
   &FN_UpdateAnimTrail(Locked Particle RibbonWidth=1, SkeletalMesh=$System.Skeletal Mesh) {module=FN_UpdateAnimTrail001}
   &ScaleColor(
    Scale Alpha=&Multiply_Float(
     A=$System.SwingGeneralFade
     B=&FloatFromCurve(FloatCurve=DataInterface() {propertyValues {Curve {keys [{arriveTangent=-0.0783, interpMode=RCIM_Cubic, leaveTangent=-0.0783, tangentMode=RCTM_User, time=0.9, value=0.9}, <at 1 0 RCIM_Cubic>]}}})
    )
    Scale RGB=&MakeVectorFromLinearColorRGB(Color=$System.TODColor)
    ScaleA=true
    ScaleRGB=true
   )
  ]
  rendererClasses [&NiagaraRibbonRendererProperties]
  renderers [{rendererClass=&NiagaraRibbonRendererProperties}]
 }
 WireStreaks {
  emitterUpdateScript [
   &EmitterState() {module=EmitterState001}
   &FN_SpawnAnimTrail()
  ]
  particleSpawnScript [
   &FN_InitializeAnimTrailBoneIndecies(Second Bone or Socket Index=2)
   &FN_InitializeAnimTrail(Lifetime=0.1, SkeletalMesh=$System.Skeletal Mesh)
  ]
  particleUpdateScript [
   &SolveForcesAndVelocity(Clamp Velocity=false, Limit Acceleration=false)
   &FN_UpdateAnimTrail(SkeletalMesh=$System.Skeletal Mesh) {module=FN_UpdateAnimTrail001}
   &ScaleColor(
    Scale Alpha=&Multiply_Float(
     A=$System.SwingGeneralFade
     B=&FloatFromCurve(
      FloatCurve=DataInterface() {propertyValues {Curve {keys [{arriveTangent=-67.7793, interpMode=RCIM_Cubic, leaveTangent=-67.7793, tangentMode=RCTM_User, time=0.12, value=1}, <at 0.15 0>]}}}
      CurveIndex=$System.Age
      Scale Curve=20
     )
    )
    ScaleA=true
    ScaleRGB=true
   ) {module=ScaleColor001}
  ]
  rendererClasses [&NiagaraRibbonRendererProperties]
  renderers [{rendererClass=&NiagaraRibbonRendererProperties}]
 }
}
summary {
 systemName=NS_Pickaxe_CrumpleFold_Trail
 emitters [
  {emitterName=SoftTrail, rendererClasses [&NiagaraRibbonRendererProperties]}
  {emitterName=ToonTrail, rendererClasses [&NiagaraRibbonRendererProperties]}
  {emitterName=WireStreaks, rendererClasses [&NiagaraRibbonRendererProperties]}
 ]
 userVariables [User.RibbonIDOffset {type=NiagaraInt32}]
}
